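# Master material for the Space Cuddles backpack cosmetic, adding an animated, tinted emissive heart screen to the standard character shading.
# Scrolling static and a sweep line distort the heart texture's UVs, then combine with distance-faded scanlines, a sine-shaped frame and view-angle falloff.
# A SkinFX mask, routed through the glitch function, isolates the emissive screen, while an FX texture and DiffuseLerp can darken base color beneath the hearts.
# Results pass through the PreFX, CharacterSkinFX and PostFX character function chain.
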Material /BRCosmetics/Accessories/FORT_Backpacks/Backpack_F_MED_Space_Cuddles/Materials/Backpack_F_MED_Space_Cuddles_Master
graph {
 properties {
  bUseMaterialAttributes=true
  bUsedWithMeshDeformer=true
  bUsedWithSkeletalMesh=true
  pixelDepthOffsetMode=PDOM_Legacy
  refractionMethod=RM_IndexOfRefraction
  usedShadingModels=Default Lit
 }
 nodes [
  n1 = TextureCoordinate()
  n2 = VectorParameter() {parameterName=TexCoordScale, defaultValue=<rgba 4 4 0 0>}
  Add(
   A=Multiply(A=@n1, B=ComponentMask(None=@n2.RGB) {g=true, r=true})
   B=ComponentMask(None=VectorParameter() {parameterName=TexCoordOffset, defaultValue=<rgba 0.0524 -1.7822 0 0>}.RGB) {g=true, r=true}
  )
  n3 = MaterialFunctionCall() {materialFunction=&BaseFN_CharacterFunction_PreFX}
  n4 = GetMaterialAttributes(None=MaterialFunctionCall(
    PreskinnedPosition=@n3.Pre skinned Local Position
    Gmap=@n3.Pre skinned Local Position
    MaterialAttributes=@n3.Pre skinned Local Position
    AO=@n3.Pre skinned Local Position
   ) {materialFunction=&CharacterSkinFX}.Result) {attributeGetTypes [#1, #2]}
  n5 = MaterialFunctionCall!(In=TextureObjectParameter() {parameterName=SkinFX_Mask}) {materialFunction=/Game/Weapons/WeaponSkins/MaterialFunction/MF_GlitchSubfunction}
  n6 = StaticSwitchParameter(True=@n5.Glitched, False=@n5.Glitched) {group=Skin FX, parameterName=UseGlitch}
  n7 = ComponentMask(None=@n6) {g=true}
  n8 = ComponentMask(None=@n6) {r=true}
  n9 = Saturate(None=Subtract(A=Subtract(A=@n7, B=@n8), B=ComponentMask(None=@n6) {b=true}))
  n10 = Multiply(A=NamedRerouteUsage()) {constB=0.5}
  n11 = Multiply(A=Sine(None=ComponentMask(None=@n10) {r=true}), B=Sine(None=ComponentMask(None=@n10) {g=true}))
  n12 = OneMinus(None=Power(Base=OneMinus(None=@n11), Exp=ScalarParameter() {defaultValue=0.8, group=Background, parameterName=Frame Exponent}))
  n13 = NamedRerouteUsage()
  n14 = ConstantBiasScale(None=Frac(None=Panner(Coordinate=@n13) {speedY=-1})) {bias=-0.5, scale=2}
  n15 = Multiply(A=@n14, B=@n14)
  n16 = Add(A=Multiply(A=ComponentMask(None=@n15) {g=true}, B=ScalarParameter() {defaultValue=0.5, group=Background, parameterName=Scanline Strength}) {constB=0.1})
  n17 = NamedRerouteUsage()
  n18 = Panner(Coordinate=@n17) {speedX=5, speedY=0.1}
  n19 = Multiply(
   A=TextureSample(UVs=Panner(Coordinate=@n17) {speedX=6, speedY=-0.05}) {samplerType=SAMPLERTYPE_Masks, texture=/Game/Missions/Primary/RepairTheShelter/Textures/Cinematic_Screen_Static}.G
   B=TextureSample(UVs=Panner(Coordinate=@n17) {speedX=-3, speedY=-0.03}) {samplerType=SAMPLERTYPE_Masks, texture=&Cinematic_Screen_Static}.R
  )
  n20 = Add(
   A=Multiply(A=TextureSample(UVs=@n18) {samplerType=SAMPLERTYPE_Masks, texture=&Cinematic_Screen_Static}.B) {constB=0.1}
   B=@n19
  )
  n21 = ScalarParameter() {defaultValue=6, group=Static, parameterName=Static Scale}
  n22 = Multiply(A=ComponentMask(None=NamedRerouteUsage()) {g=true}) {constB=0.75}
  n23 = Multiply(A=Time()) {constB=-0.2}
  n24 = Power(Base=Frac(None=Add(A=@n22, B=@n23))) {constExponent=20}
  n25 = ScalarParameter() {defaultValue=0.2, group=Sweep Line, parameterName=Sweep Add}
  n26 = Add(
   A=Add(A=Multiply(A=@n20, B=@n21), B=Multiply(A=@n24, B=@n25) {constB=0.2})
   B=ScalarParameter() {defaultValue=0.2, group=Background, parameterName=Background}
  )
  n27 = LinearInterpolate(
   A=MaterialFunctionCall!(ScanLine_Tiling_1, Scanline_PannerSpeed_1, ScanLine_Tiling, Scanline_PannerSpeed, UVs) {materialFunction=/Game/Weapons/WeaponSkins/MaterialFunction/ScanLines}.Result
   B=Constant() {r=0.5}
   Alpha=MaterialFunctionCall!(Blend Range=ScalarParameter() {defaultValue=25, parameterName=r}, Start Offset=ScalarParameter() {defaultValue=-100, parameterName=d}) {materialFunction=/Engine/Functions/Engine_MaterialFunctions03/Particles/Distance_Blend}.Result
  )
  n28 = AppendVector(A=Multiply(A=@n19, B=ScalarParameter() {defaultValue=0.07, group=Static, parameterName=Static UV Distort}), B=Constant())
  n29 = AppendVector(A=Multiply(A=@n24, B=ScalarParameter() {defaultValue=0.005, group=Sweep Line, parameterName=Sweep UV Distort}) {constB=0.02}, B=Constant())
  n30 = Multiply(A=Add(A=Add(A=@n28, B=@n29), B=TextureCoordinate()), B=ScalarParameter() {defaultValue=1, group=Colors, parameterName=HeartScale})
  n31 = ComponentMask(None=@n30) {r=true}
  n32 = ScalarParameter() {group=Colors, parameterName=HeartShiftU, sliderMax=1}
  n33 = Multiply(A=ComponentMask(None=@n30) {g=true}) {constB=-1}
  n34 = ScalarParameter() {group=Colors, parameterName=HeartShiftV, sliderMax=1}
  n35 = TextureSample(UVs=MaterialFunctionCall(X=Subtract(A=@n31, B=@n32) {constB=0.3}, Y=Subtract(A=@n33, B=@n34)) {materialFunction=&MakeFloat2}.Result) {texture=/BRCosmetics/Accessories/FORT_Backpacks/Backpack_F_MED_Space_Cuddles/Textures/hearts2_Space_Cuddles}
  n36 = Saturate(None=Add(A=Add(A=@n35.R, B=@n35.R), B=@n35.B))
  n37 = Add(A=Multiply(A=Multiply(A=@n27, B=@n36), B=ScalarParameter() {defaultValue=5, parameterName=FaceIntensity})) {constB=0}
  n38 = Normalize(VectorInput=Subtract(A=CameraPositionWS(), B=WorldPosition().XYZ))
  n39 = DotProduct(A=PixelNormalWS(), B=@n38)
  n40 = Saturate(None=Subtract(A=Abs(None=@n39)) {constB=0.8})
  n41 = ScalarParameter() {defaultValue=0.4, group=Colors, parameterName=ScreenBrightness}
  n42 = Multiply(A=Multiply(A=Multiply(A=@n12, B=@n16), B=Add(A=@n26, B=@n37)) {constA=1}, B=Saturate(None=Multiply(A=@n40, B=@n41) {constB=8}))
  n43 = Multiply(A=VectorParameter() {group=Colors, parameterName=Screen Tone, defaultValue=<rgba 0.5395 0.159 0.5776 1>}.RGB, B=@n42)
  n44 = TextureSample() {
   samplerType=SAMPLERTYPE_Masks
   texture=/BRCosmetics/Accessories/FORT_Backpacks/Backpack_F_MED_Space_Cuddles/Textures/T_F_MED_Space_Cuddles_Blue_Backpack_FX
  }
  n45 = LinearInterpolate(A=@n4.Base Color, Alpha=LinearInterpolate(B=NamedRerouteUsage(), Alpha=@n44.G)) {constB=0}
  n46 = ScalarParameter() {group=Background, parameterName=DiffuseLerp, sliderMax=1}
  MaterialFunctionCall(
   AO=@n3.AO
   Distance Blend=@n3.AO
   Fresnel=@n3.AO
   Gmap=@n3.AO
   Pre Skinned Local Position=@n3.AO
   Material=SetMaterialAttributes!(
    MaterialAttributes=@n4.MaterialAttributes
    Emissive Color=Multiply(A=@n9, B=@n43)
    Base Color=LinearInterpolate(A=@n4.Base Color, B=@n45, Alpha=@n46) {constB=0}
   ) {attributeSetTypes [#1, #2]}
  ) {materialFunction=&BaseFN_CharacterFunction_PostFX}
  MaterialFunctionCall() {materialFunction=&CameraDirectionVector}
  Multiply(A=@n38) {constB=-1}
  NamedRerouteDeclaration(None=Saturate(None=@n42)) {name=HeartMask, nodeColor=<rgba 1 0.3281 0 1>}
  NamedRerouteDeclaration(None=@n1) {name=TexCoord, nodeColor=<rgba 1 0 0.4453 1>}
  TextureCoordinate() {vTiling=15}
  TextureCoordinate()
 ]
}
