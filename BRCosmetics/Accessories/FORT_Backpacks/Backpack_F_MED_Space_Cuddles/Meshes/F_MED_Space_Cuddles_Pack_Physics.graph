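# Physics asset for the Space Cuddles backpack cosmetic, covering its left and right dynamic wing bones.
# Each side pairs a capsule on the master wing bone with a smaller capsule on its matching Backpack bone.
# One constraint per side joins the pair, with twist and one swing axis nearly locked so the wing flexes mainly along the other swing axis.
# The two sides are mirrored and have no constraint between them.
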PhysicsAsset /BRCosmetics/Accessories/FORT_Backpacks/Backpack_F_MED_Space_Cuddles/Meshes/F_MED_Space_Cuddles_Pack_Physics
bodies {
 Dyn_MasterLeftWing [{shapeType=Capsule, length=1, radius=1, center=<0.5 0 0>, rotation=<rot 0 0 90>}]
 Dyn_MasterLeftWing_Backpack [{shapeType=Capsule, length=0.505, radius=0.505, rotation=<rot 0 -90 0>}]
 Dyn_MasterRightWing [{shapeType=Capsule, length=1, radius=1, center=<-0.5 0 0>, rotation=<rot 0 0 90>}]
 Dyn_MasterRightWing_Backpack [{shapeType=Capsule, length=0.505, radius=0.505, rotation=<rot 0 -90 0>}]
}
constraints [
 {bone1Name=Dyn_MasterLeftWing, bone2Name=Dyn_MasterLeftWing_Backpack, swing1LimitDegrees=1, swing2LimitDegrees=20, twistLimitDegrees=1}
 {bone1Name=Dyn_MasterRightWing, bone2Name=Dyn_MasterRightWing_Backpack, swing1LimitDegrees=1, swing2LimitDegrees=20, twistLimitDegrees=1}
]
