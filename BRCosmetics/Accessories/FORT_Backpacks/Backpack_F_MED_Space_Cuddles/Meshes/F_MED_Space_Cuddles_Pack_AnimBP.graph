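# Animation Blueprint for the Space Cuddles backpack, posing its wings from the player's movement state.
# While skydiving, the Flight_Poses state machine plays an airborne blendspace when accelerating or a floating loop otherwise, layered with partial-strength RigidBody physics.
# Otherwise, a Grounded pose blends fixed wing bone rotations for sprinting, jumping and falling.
# The event graph feeds the blendspace's Left Right from PlayerMovement and zeroes Forward; several nodes are left unconnected.
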AnimBlueprint /BRCosmetics/Accessories/FORT_Backpacks/Backpack_F_MED_Space_Cuddles/Meshes/F_MED_Space_Cuddles_Pack_AnimBP
target_skeleton=&/F_MED_Space_Cuddles_Pack_Skeleton
graphs [
 AnimGraph {nodes [
   RigidBody() {
    bOverrideWorldGravity=true
    overridePhysicsAsset=&/F_MED_Space_Cuddles_Pack_Physics
    alphaBoolBlend {blendOption=Linear}
    overrideWorldGravity {z=-980}
   }
   n1 = RigidBody!(
    ComponentPose=Local To Component(LocalPose=Use cached pose 'FlightPosesCache'())
    Alpha=0.4
    OverrideWorldGravity:Vector="0,0,-500"
    ExternalForce
   ) -> (Pose) {
    bClampLinearTranslationLimitToRefPose=true
    bOverrideWorldGravity=true
    overridePhysicsAsset=&F_MED_Space_Cuddles_Pack_Physics
    simulationSpace=BaseBoneSpace
    alphaBoolBlend {blendOption=Linear}
    alphaScaleBiasClamp {bInterpResult=true, interpSpeedDecreasing=30, interpSpeedIncreasing=30}
    baseBoneRef {boneName=root}
    componentAppliedLinearAccClamp=<5000 5000 5000>
    componentLinearAccScale=<0.25 0.25 0.25>
    componentLinearVelScale=<0.4 0.4 0.4>
   }
   Output Pose(Result=Blend Poses by bool(BlendPose_0=Component To Local(ComponentPose=@n1), BlendPose_1=Use cached pose 'Grounded'(), bActiveValue=Get bIsSkyDiving()) {blendType=Linear}) {layerGroup=None}
   AnimGraphNode_StateMachine_0 = Flight_Poses() {is=title AnimGraphNode_StateMachine}
   Save cached pose 'FlightPosesCache'(Pose=@AnimGraphNode_StateMachine_0) {cacheName=FlightPosesCache}
   n2 = "Transform (Modify) Bone - Bone: Dyn_MasterLeftWing"(Alpha=1, Translation="0,0,0", Rotation="30,0,0", Scale="1,1,1") {
    rotationMode=BMM_Replace
    rotationSpace=BCS_BoneSpace
    alphaBoolBlend {blendOption=Linear}
    boneToModify {boneName=Dyn_MasterLeftWing}
    rotation {pitch=30}
   }
   n3 = "Transform (Modify) Bone - Bone: Dyn_MasterRightWing"(ComponentPose=@n2, Alpha=1, Translation="0,0,0", Rotation="30,0,0", Scale="1,1,1") {
    rotationMode=BMM_Replace
    rotationSpace=BCS_BoneSpace
    alphaBoolBlend {blendOption=Linear}
    boneToModify {boneName=Dyn_MasterRightWing}
    rotation {pitch=30}
   }
   n4 = "Transform (Modify) Bone - Bone: Dyn_MasterLeftWing"(Alpha=1, Translation="0,0,0", Rotation="30,0,0", Scale="1,1,1") {
    rotationMode=BMM_Replace
    rotationSpace=BCS_BoneSpace
    alphaBoolBlend {blendOption=Linear}
    boneToModify {boneName=Dyn_MasterLeftWing}
    rotation {pitch=30}
   }
   n5 = "Transform (Modify) Bone - Bone: Dyn_MasterRightWing"(ComponentPose=@n4, Alpha=1, Translation="0,0,0", Rotation="30,0,0", Scale="1,1,1") {
    rotationMode=BMM_Replace
    rotationSpace=BCS_BoneSpace
    alphaBoolBlend {blendOption=Linear}
    boneToModify {boneName=Dyn_MasterRightWing}
    rotation {pitch=30}
   }
   n6 = "Transform (Modify) Bone - Bone: Dyn_MasterLeftWing"(ComponentPose=Local To Component(LocalPose=Input Pose()), Alpha=1, Translation="0,0,0", Rotation="0,0,0", Scale="1,1,1") {rotationMode=BMM_Replace, rotationSpace=BCS_BoneSpace, alphaBoolBlend {blendOption=Linear}, boneToModify {boneName=Dyn_MasterLeftWing}}
   n7 = "Transform (Modify) Bone - Bone: Dyn_MasterRightWing"(ComponentPose=@n6, Alpha=1, Translation="0,0,0", Rotation="0,0,0", Scale="1,1,1") {rotationMode=BMM_Replace, rotationSpace=BCS_BoneSpace, alphaBoolBlend {blendOption=Linear}, boneToModify {boneName=Dyn_MasterRightWing}}
   n8 = "Transform (Modify) Bone - Bone: Dyn_MasterLeftWing"(Alpha=1, Translation="0,0,0", Rotation="-30,0,0", Scale="1,1,1") {
    rotationMode=BMM_Replace
    rotationSpace=BCS_BoneSpace
    alphaBoolBlend {blendOption=Linear}
    boneToModify {boneName=Dyn_MasterLeftWing}
    rotation {pitch=-30}
   }
   n9 = "Transform (Modify) Bone - Bone: Dyn_MasterRightWing"(ComponentPose=@n8, Alpha=1, Translation="0,0,0", Rotation="-30,0,0", Scale="1,1,1") {
    rotationMode=BMM_Replace
    rotationSpace=BCS_BoneSpace
    alphaBoolBlend {blendOption=Linear}
    boneToModify {boneName=Dyn_MasterRightWing}
    rotation {pitch=-30}
   }
   n10 = Two Way Blend!(A=Component To Local(ComponentPose=@n7), B=Component To Local(ComponentPose=@n9), bAlphaBoolEnabled:bool=Get bIsSprinting()) -> (Pose) {blendNode {
     alphaInputType=Bool
     alphaBoolBlend {blendInTime=0.5, blendOption=Linear, blendOutTime=1}
     alphaScaleBiasClamp {bInterpResult=true, interpSpeedIncreasing=0.1, inRange {min=-0.0005}, outRange {max=24}}
    }}
   n11 = Two Way Blend(A=@n10, B=Component To Local(ComponentPose=@n5), bAlphaBoolEnabled=Get bIsJumping()) {blendNode {
     alphaInputType=Bool
     bResetChildOnActivation=true
     alphaBoolBlend {blendInTime=0.5, blendOption=Linear, blendOutTime=1}
     alphaScaleBiasClamp {bInterpResult=true, interpSpeedIncreasing=0.1, inRange {min=-0.0005}, outRange {max=24}}
    }}
   n12 = Two Way Blend(A=@n11, B=Component To Local(ComponentPose=@n3), bAlphaBoolEnabled=Get bIsFalling()) {blendNode {
     alphaInputType=Bool
     bResetChildOnActivation=true
     alphaBoolBlend {blendInTime=0.5, blendOption=Linear, blendOutTime=1}
     alphaScaleBiasClamp {bInterpResult=true, interpSpeedIncreasing=0.1, inRange {min=-0.0005}, outRange {max=24}}
    }}
   Save cached pose 'Grounded'(Pose=@n12) {cacheName=Grounded}
   n13 = "Transform (Modify) Bone - Bone: Dyn_MasterLeftWing"(Alpha=1, Translation="0,0,0", Rotation="-15,0,0", Scale="1,1,1") {
    rotationMode=BMM_Replace
    rotationSpace=BCS_BoneSpace
    alphaBoolBlend {blendOption=Linear}
    boneToModify {boneName=Dyn_MasterLeftWing}
    rotation {pitch=-15}
   }
   n14 = "Transform (Modify) Bone - Bone: Dyn_MasterRightWing"(ComponentPose=@n13, Alpha=1, Translation="0,0,0", Rotation="-15,0,0", Scale="1,1,1") {
    rotationMode=BMM_Replace
    rotationSpace=BCS_BoneSpace
    alphaBoolBlend {blendOption=Linear}
    boneToModify {boneName=Dyn_MasterRightWing}
    rotation {pitch=-15}
   }
   Two Way Blend(B=Component To Local(ComponentPose=@n14), bAlphaBoolEnabled=True) {blendNode {
     alphaInputType=Bool
     bResetChildOnActivation=true
     alphaBoolBlend {blendInTime=0.5, blendOption=Linear, blendOutTime=1}
     alphaScaleBiasClamp {bInterpResult=true, interpSpeedIncreasing=0.1, inRange {min=-0.0005}, outRange {max=24}}
    }}
   "To Float (Boolean)"(InBool=OR Boolean!(A=Get bIsCrouching(), B=Get bIsCrouchSprinting(), C:bool=Get bIsTargeting(), D:bool=Get bIsOnGround() -> (bIsOnGround:bool)) -> (ReturnValue))
   SelectVector(A="0, 0,-1000", B="0,0,-750", bPickA=Get bIsOnGround())
   Get bIdle_Walk_Transition!(self:Fort Item Layer Anim Instance Galileo Patrol Object Reference) -> (bIdle_Walk_Transition:bool)
  ]}
 Flight_Poses {
  entry=Base
  path=AnimGraph.AnimGraphNode_StateMachine_0.Flight_Poses
  ids=id
  nodes [
   AnimStateNode_0 = Base() {is=title AnimStateNode}
   AnimStateNode_1 = Skydiving() {
    is=title AnimStateNode
    stateEntered {guid=#1}
    stateFullyBlended {guid=#2}
    stateLeft {guid=#3}
   }
   AnimStateTransitionNode_0 = AnimStateTransitionNode()
   AnimStateTransitionNode_1 = AnimStateTransitionNode()
  ]
  links [
   {
    dst=Skydiving
    rule=AnimGraph.AnimGraphNode_StateMachine_0.Flight_Poses.AnimStateTransitionNode_0.Transition
    src=Base
    via=AnimStateTransitionNode_0
   }
   {
    dst=Base
    rule=AnimGraph.AnimGraphNode_StateMachine_0.Flight_Poses.AnimStateTransitionNode_1.Transition
    src=Skydiving
    via=AnimStateTransitionNode_1
   }
  ]
 }
 Base {
  path=AnimGraph.AnimGraphNode_StateMachine_0.Flight_Poses.AnimStateNode_0.Base
  nodes [Output Animation Pose(Result=Sequence Player 'Space_Cuddles_Air_Base_Anim'() {sequence=/BRCosmetics/Accessories/FORT_Backpacks/Backpack_F_MED_Space_Cuddles/Animations/Space_Cuddles_Air_Base_Anim})]
 }
 Skydiving {
  path=AnimGraph.AnimGraphNode_StateMachine_0.Flight_Poses.AnimStateNode_1.Skydiving
  nodes [
   n1 = Get Left Right() -> (Left Right:double)
   n2 = "Get Forward "() -> ("Forward ":double)
   Output Animation Pose(Result=Blend Poses by bool(
     BlendPose_0=Blendspace Player 'F_MED_Space_Cuddles_Airborne_BS'(X=@n1, Y=@n2) {blendSpace=/BRCosmetics/Accessories/FORT_Backpacks/Backpack_F_MED_Space_Cuddles/Animations/F_MED_Space_Cuddles_Airborne_BS}
     BlendPose_1=Sequence Player 'Space_Cuddles_Air_Floating_Anim'() {sequence=/BRCosmetics/Accessories/FORT_Backpacks/Backpack_F_MED_Space_Cuddles/Animations/Space_Cuddles_Air_Floating_Anim}
     BlendTime_0=0.3
     BlendTime_1=0.3
     bActiveValue=Get bIsAccelerating2D()
    ) {blendType=Linear})
  ]
 }
 Transition {
  path=AnimGraph.AnimGraphNode_StateMachine_0.Flight_Poses.AnimStateTransitionNode_0.Transition
  nodes [Result(bCanEnterTransition=Get bIsSkyDiving())]
 }
 Transition {
  path=AnimGraph.AnimGraphNode_StateMachine_0.Flight_Poses.AnimStateTransitionNode_1.Transition
  nodes [
   Result(bCanEnterTransition=NOT Boolean(A=Get bIsSkyDiving()))
   Get bIsSkydiveFloating()
  ]
 }
 EventGraph {nodes [
   Event BlueprintInitializeAnimation()
   > Set Skeletal Mesh(Skeletal Mesh:Skeletal Mesh Object Reference=&/F_MED_Space_Cuddles_Pack) -> (Output_Get:Skeletal Mesh Object Reference)
   > n1 = Cast To FortPlayerPawn(Object=TryGetFortPawn())
   > Set Pawn(Pawn:Fort Player Pawn Object Reference=@n1) -> (Output_Get:Fort Player Pawn Object Reference)
   Event BlueprintUpdateAnimation()
   > Get!(execute:Exec) -> (then:Exec, else:Exec, Skeletal Mesh:Skeletal Mesh Object Reference)
   > Get!(execute) -> (then, else, Pawn:Fort Player Pawn Object Reference)
   > n2 = PlayerMovement(AnimInstance=Self-Reference(), XDotProductCoefficient=25, YDotProductCoefficient=25, ZDotProductCoefficient=1)
   > n3 = Sequence!(execute) -> (then_0, then_1, then_2:Exec)
   > Set Left Right(Left Right:double=@n2) -> (Output_Get:double)
   "Set Forward "(execute=@n3.then_1, "Forward ":double=0) -> (Output_Get:double)
   Set Move Backwards(execute=@n3.then_2, Move Backwards:double=@n2.YMovement) -> (Output_Get:double)
   n4 = Branch()
   > Do Once(Reset=@n6.then)
   > Set Backpack Start State(Backpack Start State:double=1) -> (Output_Get:double)
   > Delay(Duration=0.2)
   > n5 = Set Moving Forward(Moving Forward:double=0) -> (Output_Get:double)
   Do Once(execute=@n4.else, Reset=@n5.then, Start Closed=true)
   > Set Backpack Start State(Backpack Start State=0)
   > Delay(Duration=0.2)
   > n6 = Set Moving Forward(Moving Forward=1)
   TryGetPawnOwner()
   "To Float (Boolean)"(InBool=Get bIsSprinting())
   Branch()
   Get bIsSkyDiving()
  ]}
]
variables [
 Backpack Start State {type=number}
 "Forward " {type=number}
 Left Right {type=number}
 Move Backwards {type=number}
 Moving Forward {type=number}
 Pawn {type=FortPlayerPawn}
 Skeletal Mesh {type=SkeletalMesh, default=&F_MED_Space_Cuddles_Pack}
 Skeletal Mesh_0 {type=SkeletalMesh, default=null}
 Sky Diving {type=number}
 Sky Floating {type=number}
 Sky Left {type=number}
 Sky Right {type=number}
]
